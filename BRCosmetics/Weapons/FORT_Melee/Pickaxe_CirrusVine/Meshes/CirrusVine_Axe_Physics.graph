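# Physics asset for the CirrusVine two-handed pickaxe mesh, adding a physically constrained rag chain to the weapon.
# Two capsules cover the handle bone, and single capsules along five dyn_Rag bones plus an end bone form a chain attached to it.
# Swing limits are widest where the rag joins the handle and tighten down the chain, twist stays nearly locked, and the end bone allows no swing or twist.
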PhysicsAsset /BRCosmetics/Weapons/FORT_Melee/Pickaxe_CirrusVine/Meshes/CirrusVine_Axe_Physics
bodies {
 Melee_TwoHanded_Handle [
  {shapeType=Capsule, length=80.6706, radius=14.9132, center=<0.6668 5.4883 32.8475>, rotation=<rot 1.2966 -90 0.0136>}
  {shapeType=Capsule, length=24.8522, radius=2.2382, shapeName=Melee_TwoHanded_Handle_capsule, center=<0.6775 0 -41.403>}
 ]
 dyn_Rag_0 [{shapeType=Capsule, length=1.5667, radius=1.0653, center=<1.4744 -0.1898 0.5395>, rotation=<rot -3.9319 4.9291 -89.9369>}]
 dyn_Rag_1 [{
   shapeType=Capsule
   length=2.5142
   radius=1.4465
   shapeName=dyn_Rag_1_capsule
   center=<1.0981 -0.1044 -0.4404>
   rotation=<rot 0 5.5465 -89.296>
  }]
 dyn_Rag_2 [{shapeType=Capsule, length=3.2293, radius=1.3789, shapeName=dyn_Rag_2_capsule, center=<0.6414 0.0954 -0.7054>, rotation=<rot 0 3.699 -88.809>}]
 dyn_Rag_3 [{
   shapeType=Capsule
   length=4.4391
   radius=1.6532
   shapeName=dyn_Rag_3_capsule
   center=<0.7785 0.3356 -0.937>
   rotation=<rot 0 -3.1066 -88.7252>
  }]
 dyn_Rag_4 [{
   shapeType=Capsule
   length=5.8619
   radius=2.4008
   shapeName=dyn_Rag_4_capsule
   center=<1.7275 0.4988 -1.2612>
   rotation=<rot 0 -2.9603 -89.4053>
  }]
 dyn_Rag_end [{shapeType=Capsule, length=0.505, radius=0.505, shapeName=dyn_Rag_end_capsule, rotation=<rot 0 -90 0>}]
}
constraints [
 {bone1Name=dyn_Rag_0, bone2Name=Melee_TwoHanded_Handle, swing1LimitDegrees=50, swing2LimitDegrees=70.32, twistLimitDegrees=1}
 {bone1Name=dyn_Rag_1, bone2Name=dyn_Rag_0, swing1LimitDegrees=14, swing2LimitDegrees=15, twistLimitDegrees=1}
 {bone1Name=dyn_Rag_2, bone2Name=dyn_Rag_1, swing1LimitDegrees=14, swing2LimitDegrees=15, twistLimitDegrees=1}
 {bone1Name=dyn_Rag_3, bone2Name=dyn_Rag_2, swing1LimitDegrees=10, swing2LimitDegrees=12, twistLimitDegrees=1}
 {bone1Name=dyn_Rag_4, bone2Name=dyn_Rag_3, swing1LimitDegrees=8, swing2LimitDegrees=11, twistLimitDegrees=1}
 {bone1Name=dyn_Rag_end, bone2Name=dyn_Rag_4, swing1LimitDegrees=0, swing2LimitDegrees=0, twistLimitDegrees=0}
]
