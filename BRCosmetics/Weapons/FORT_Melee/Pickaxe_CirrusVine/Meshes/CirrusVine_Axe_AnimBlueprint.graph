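# Animation blueprint for the CirrusVine pickaxe, a two-handed melee axe, adding secondary physics motion to its skeleton.
# After the DefaultSlot montage, it offsets the blade and handle bones, then runs a RigidBody simulation in handle bone space with custom gravity.
# Anim Dynamics and a four-bone Trail controller follow on the dyn_Rag bones.
# Their shared alpha is set each update to the inverse of bRigidBodyEnabled, making them a fallback when rigid body is off.
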AnimBlueprint /BRCosmetics/Weapons/FORT_Melee/Pickaxe_CirrusVine/Meshes/CirrusVine_Axe_AnimBlueprint
target_skeleton=&/CirrusVine_Axe_Skeleton
graphs [
 AnimGraph {nodes [
   n1 = Input Pose()
   n2 = "Transform (Modify) Bone - Bone: blade_jnt"(ComponentPose=Local To Component(LocalPose=Slot 'DefaultSlot'(Source=@n1)), Alpha=1, Translation="0,0,0", Rotation="0,150,0", Scale="1,1,1") {rotationMode=BMM_Additive, alphaBoolBlend {blendOption=Linear}, boneToModify {boneName=blade_jnt}, rotation {yaw=150}}
   n3 = "Transform (Modify) Bone - Bone: Melee_TwoHanded_Handle"(ComponentPose=@n2, Alpha=1, Translation="0,0,18", Rotation="0,0,0", Scale="1,1,1") {translationMode=BMM_Additive, alphaBoolBlend {blendOption=Linear}, boneToModify {boneName=Melee_TwoHanded_Handle}, translation {z=18}}
   n4 = RigidBody(ComponentPose=@n3) {
    bOverrideWorldGravity=true
    cachedBoundsScale=2
    overridePhysicsAsset=&/CirrusVine_Axe_Physics
    simulationSpace=BaseBoneSpace
    alphaBoolBlend {blendOption=Linear}
    baseBoneRef {boneName=Melee_TwoHanded_Handle}
    componentAppliedLinearAccClamp=<1000 1000 1000>
    componentLinearAccScale=<0.5 0.5 0.5>
    componentLinearVelScale=<0.5 0.5 0.5>
    overrideWorldGravity {z=-500}
    simSpaceSettings {maxAngularAcceleration=300, maxAngularVelocity=3, maxLinearAcceleration=200, maxLinearVelocity=300, velocityScaleZ=0, worldAlpha=1}
   }
   n5 = Anim Dynamics - Bone: dyn_Rag_0(ComponentPose=@n4, Alpha=Get rigid_body_enabled() -> (rigid_body_enabled:double), ExternalForce="0,0,0") {
    angularBiasOverride=0.95
    angularDampingOverride=0.95
    linearDampingOverride=0.95
    simulationSpace=RootRelative
    alphaBoolBlend {blendOption=Linear}
    boundBone {boneName=dyn_Rag_0}
    componentLinearAccScale=<0.3 0.3 0.3>
    componentLinearVelScale=<0.3 0.3 0.3>
    physicsBodyDefinitions [{
      boundBone {boneName=dyn_Rag_0}
      boxExtents=<4 13 4>
      constraintSetup {angularLimitsMax=<30 60 80>, angularLimitsMin=<-30 -65 -80>, angularTarget=<0 0 0>}
      localJointOffset=<2 0 0>
     }]
   }
   n6 = Trail controller - Bone: dyn_Rag_4(ComponentPose=@n5, Alpha=Get rigid_body_enabled()) {
    bActorSpaceFakeVel=true
    bLimitRotation=true
    bUsePlanarLimit=true
    chainLength=4
    relaxationSpeedScale=1.7
    fakeVelocity=<1 1 1>
    planarLimits [{drivingBone {boneName=b_AR_Root}, planeTransform {location=<0 0 -50>, rotation=<rot 0 0 -90>, scale=<1 1 1>}}]
    rotationLimits [
     {limitMax=<25 50 35>, limitMin=<-25 -50 -35>}
     {limitMax=<25 50 35>, limitMin=<-25 -50 -35>}
     {limitMax=<25 50 35>, limitMin=<-25 -50 -35>}
     {limitMax=<25 50 35>, limitMin=<-25 -50 -35>}
    ]
    rotationOffsets [<0 0 0>, <0 0 0>, <0 0 0>, <0 0 0>]
    trailBone {boneName=dyn_Rag_4}
   }
   Output Pose(Result=Component To Local(ComponentPose=@n6)) {layerGroup=None}
  ]}
 EventGraph {nodes [
   Event BlueprintUpdateAnimation()
   > n1 = Branch(Condition=Get bRigidBodyEnabled())
   > Set rigid_body_enabled(rigid_body_enabled:double=0) -> (Output_Get:double)
   Set rigid_body_enabled(execute=@n1.else, rigid_body_enabled=1)
  ]}
]
variables [rigid_body_enabled {type=number}]
